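# UI sound cue for the whoosh played, by its name, when the mode select window appears in the front-end Play tab.
# A mixer layers a party-invite appear wave, pitched up to double by its modulator, with a medium Battle Pass modal whoosh through a second, default modulator.
# It plays quietly in the Fort_UI sound class, with a concurrency override that prevents new instances and sets a short retrigger time.
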SoundCue /Game/Sounds/Fort_UI_Sounds/FrontEnd/Play_Tab/UI_Whoosh_ModeSelectWindowAppear_Cue
root=Mixer_0
nodes [
 {id=Mixer_0, children [Modulator_0,Modulator_1], properties {inputVolume [1,1]}}
 {id=Modulator_0, children [WavePlayer_1], properties {pitchMax=2, pitchMin=2}}
 {id=Modulator_1, children [WavePlayer_0]}
 {id=WavePlayer_0, properties {soundWaveAssetPtr=/Game/Sounds/Fort_UI_Sounds/FrontEnd/BattlePass/BP_ToolKit/BP_Modal_Whoosh_Med_02}}
 {id=WavePlayer_1, properties {soundWaveAssetPtr=/Game/Sounds/Fort_UI_Sounds/FrontEnd/ItemShop/wav/UI_PartyInvite_Appear}}
]
properties {
 bOverrideConcurrency=true
 maxDistance=2097152
 soundClassObject=&Fort_UI
 volumeMultiplier=0.1
 concurrencyOverrides {resolutionRule=PreventNew, retriggerTime=0.2}
}
